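# Physics asset for the one-handed Tactical Fisherman pickaxe, defining collision and constraints for its dynamic rope bones.
# Three box bodies form a chain from rope_01 to rope_03, anchored to a smaller box on the rope master bone.
# The first joint allows wide swing with tight twist, while the two lower joints share equal, moderate swing and twist limits.
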PhysicsAsset /BRCosmetics/Weapons/FORT_Melee/Pickaxe_Tactical_Fisherman_1h/Meshes/Tactical_Fisherman_1h_Physics
bodies {
 dyn_rope_01 [{shapeType=Box, extentX=1.1, extentY=4, extentZ=4, center=<1.7742 0 0>, rotation=<rot 90 0 0>}]
 dyn_rope_02 [{shapeType=Box, extentX=1.1, extentY=4, extentZ=4, center=<2.1962 0 0>, rotation=<rot 90 0 0>}]
 dyn_rope_03 [{shapeType=Box, extentX=1.1, extentY=4, extentZ=4, center=<2.8399 0 0>, rotation=<rot 90 0 0>}]
 dyn_rope_mstr [{shapeType=Box, extentX=1.01, extentY=1.01, extentZ=1.01}]
}
constraints [
 {bone1Name=dyn_rope_01, bone2Name=dyn_rope_mstr, swing1LimitDegrees=70, swing2LimitDegrees=70, twistLimitDegrees=10}
 {bone1Name=dyn_rope_02, bone2Name=dyn_rope_01, swing1LimitDegrees=45, swing2LimitDegrees=45, twistLimitDegrees=45}
 {bone1Name=dyn_rope_03, bone2Name=dyn_rope_02, swing1LimitDegrees=45, swing2LimitDegrees=45, twistLimitDegrees=45}
]
